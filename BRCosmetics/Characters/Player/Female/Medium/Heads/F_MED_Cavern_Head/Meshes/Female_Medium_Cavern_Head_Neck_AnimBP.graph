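# Animation Blueprint for the Female Medium Cavern head's neck on the Fortnite base head skeleton, narrowing the neck by scaling two dyn_neck bones.
# The input pose passes through two chained Modify Bone nodes that shrink Y and Z on dyn_neck_01 and dyn_neck_02.
# The first scales additively, while the second replaces scale in bone space.
# The locomotion, facial and mesh-copy variables are unused in the graphs shown, and the event graph has no logic.
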AnimBlueprint /BRCosmetics/Characters/Player/Female/Medium/Heads/F_MED_Cavern_Head/Meshes/Female_Medium_Cavern_Head_Neck_AnimBP
target_skeleton=/Game/Characters/Player/Common/Fortnite_Base_Head/Fortnite_Base_Head_Export_Skeleton
graphs [
 AnimGraph {nodes [
   n1 = "Transform (Modify) Bone - Bone: dyn_neck_01"(ComponentPose=Local To Component(LocalPose=Input Pose()), Alpha=1, Translation="0,0,0", Rotation="0,0,0", Scale="1,0.6,0.6") {
    scaleMode=BMM_Additive
    translationSpace=BCS_BoneSpace
    alphaBoolBlend {blendOption=Linear}
    boneToModify {boneName=dyn_neck_01}
    scale {y=0.6, z=0.6}
   }
   n2 = "Transform (Modify) Bone - Bone: dyn_neck_02"(ComponentPose=@n1, Alpha=1, Translation="0,0,0", Rotation="0,0,0", Scale="1,0.6,0.6") {
    scaleMode=BMM_Replace
    scaleSpace=BCS_BoneSpace
    translationSpace=BCS_BoneSpace
    alphaBoolBlend {blendOption=Linear}
    boneToModify {boneName=dyn_neck_02}
    scale {y=0.6, z=0.6}
   }
   Output Pose(Result=Component To Local(ComponentPose=@n2)) {layerGroup=None}
  ]}
 EventGraph {nodes [Event BlueprintInitializeAnimation()]}
]
variables [
 bIsCrouching_0 {type=boolean}
 bIsCrouchSprinting_0 {type=boolean}
 bIsCrouchWalking {type=boolean}
 bIsDancing {type=boolean}
 bIsDBNO_0 {type=boolean}
 bIsFallingLocal {type=boolean}
 bIsJogging_0 {type=boolean}
 bIsJumpingLocal {type=boolean}
 bIsLocomotion_0 {type=boolean}
 bIsMeleeSwing_0 {type=boolean}
 bIsMoving {type=boolean}
 bIsOnGround_0 {type=boolean}
 bIsSlopeSliding_0 {type=boolean}
 bIsSprinting_0 {type=boolean}
 bIsTargeting_0 {type=boolean}
 bRecentlyFired_0 {type=boolean}
 bypassIdleAnim {type=boolean}
 facial_curve_mapping_expression {type=CurveExpressionsDataAsset, default=/Game/Characters/Player/Common/Fortnite_Base_Head/Facials/CurveMappings/FN_3LToLegacy_Main_Mapping}
 Facial_Curve_Zero_Expression {
  type=CurveExpressionsDataAsset
  default=/Game/Characters/Player/Common/Fortnite_Base_Head/Facials/CurveMappings/FN_ZeroOutLegacyDefaultFaceIdle
 }
 flipbook_face_activation_var {type=FN_Flipbook_Face_Region_Activation, default {enable_Brow=true, enable_Eye=true, enable_Mouth=true}}
 GetPlayerListUsesCustomCharacterParts {type=boolean}
 is_3L_anim_obsolete {type=boolean}
 IsDive0 {type=boolean}
 IsSkydiving0 {type=boolean}
 LastFireTime {type=number}
 LookPitch {type=number}
 LookYaw {type=number}
 MeshToCopy {type=SkeletalMeshComponent}
]
